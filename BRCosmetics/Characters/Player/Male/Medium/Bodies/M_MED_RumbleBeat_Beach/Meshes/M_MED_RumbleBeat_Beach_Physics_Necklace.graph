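# Physics asset that simulates the necklace on the medium male RumbleBeat Beach player body cosmetic.
# A center chain hangs from spine_05 on an angularly free joint, with tightly to moderately limited capsule links ending in two boxes.
# Short left and right chains run from base bones into center link 04, asymmetrically: left angularly free, right limited.
# Spine, pec and clavicle helper capsules supply torso shapes, apparently for the necklace to collide against.
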PhysicsAsset /BRCosmetics/Characters/Player/Male/Medium/Bodies/M_MED_RumbleBeat_Beach/Meshes/M_MED_RumbleBeat_Beach_Physics_Necklace
bodies {
 base_necklace_03_lf [{shapeType=Capsule, length=0.505, radius=0.505, shapeName=base_necklace_03_lf_capsule, rotation=<rot 0 -90 0>}]
 base_necklace_03_rt [{shapeType=Capsule, length=0.505, radius=0.505, shapeName=base_necklace_03_rt_capsule, rotation=<rot 0 -90 0>}]
 clavicle_l_physics_help [{shapeType=Capsule, length=16, radius=5.5, shapeName=clavicle_l_capsule, center=<7.25 3.55 -4.35>, rotation=<rot 0 0 -50>}]
 clavicle_r_physics_help [{shapeType=Capsule, length=16, radius=5.5, shapeName=clavicle_l_capsule, center=<-7.25 -3.55 4.35>, rotation=<rot 0 0 -50>}]
 deform_pec_l [{shapeType=Capsule, length=10, radius=4.2, shapeName=deform_pec_l_capsule, center=<-4.7 -0.9 -4.7>, rotation=<rot 65 25 0>}]
 deform_pec_r [{shapeType=Capsule, length=10, radius=4.4, shapeName=deform_pec_r_capsule, center=<4.7 0.9 4.7>, rotation=<rot 65 25 0>}]
 dyn_necklace_01_cr [{shapeType=Capsule, length=10.4034, radius=1.3984, shapeName=dyn_necklace_01_cr_capsule, center=<0.0001 0 0.0214>, rotation=<rot 0 0 -90>}]
 dyn_necklace_02_cr [{shapeType=Capsule, length=5, radius=2.5, shapeName=dyn_necklace_02_cr_capsule, center=<4 0 0>, rotation=<rot 90 0 0>}]
 dyn_necklace_03_cr [{shapeType=Capsule, length=16, radius=2.1, shapeName=dyn_necklace_02_cr_capsule, center=<3 0 0>, rotation=<rot 0 0 90>}]
 dyn_necklace_03_lf [{shapeType=Capsule, length=2, radius=2, shapeName=dyn_necklace_03_lf_capsule, center=<2 0 0>, rotation=<rot 90 0 0>}]
 dyn_necklace_03_rt [{shapeType=Capsule, length=2, radius=2, shapeName=dyn_necklace_03_rt_capsule, center=<2 0 0>, rotation=<rot 90 0 0>}]
 dyn_necklace_04_cr [{shapeType=Capsule, length=4, radius=2, shapeName=dyn_necklace_02_cr_capsule, rotation=<rot 0 0 90>}]
 dyn_necklace_04_lf [{shapeType=Capsule, length=2, radius=2, shapeName=dyn_necklace_04_lf_capsule, center=<2 0 0>, rotation=<rot 90 0 0>}]
 dyn_necklace_04_rt [{shapeType=Capsule, length=2, radius=2, shapeName=dyn_necklace_04_rt_capsule, center=<2 0 0>, rotation=<rot 90 0 0>}]
 dyn_necklace_05_cr [{shapeType=Box, extentX=8, extentY=6, extentZ=1, shapeName=dyn_necklace_05_cr_box, center=<4.1 0 1.8>, rotation=<rot 10.0002 0 0>}]
 dyn_necklace_06_cr [{
   shapeType=Box
   extentX=4.5
   extentY=7
   extentZ=1.8
   shapeName=dyn_necklace_05_cr_box
   center=<9.1 -2.7 -0.15>
   rotation=<rot -2.2812 -13.8421 37.0243>
  }]
 spine_05 [
  {shapeType=Capsule, length=22, radius=10.6416, center=<-4.6117 2.7212 -0.0229>, rotation=<rot 0 64.3566 0.2455>}
  {shapeType=Capsule, length=16, radius=10, shapeName=spine_05_capsule, center=<0.4663 -1.6405 0>}
 ]
}
constraints [
 {bone1Name=dyn_necklace_03_cr, bone2Name=dyn_necklace_02_cr, swing1LimitDegrees=2, swing2LimitDegrees=10, twistLimitDegrees=1}
 {bone1Name=dyn_necklace_04_cr, bone2Name=dyn_necklace_04_rt, swing1LimitDegrees=2, swing2LimitDegrees=15, twistLimitDegrees=3}
 {bone1Name=dyn_necklace_05_cr, bone2Name=dyn_necklace_04_cr, swing1LimitDegrees=15, swing2LimitDegrees=10, twistLimitDegrees=5}
 {bone1Name=dyn_necklace_02_cr, bone2Name=dyn_necklace_01_cr, swing1LimitDegrees=3, swing2LimitDegrees=20, twistLimitDegrees=1}
 {
  bone1Name=dyn_necklace_01_cr
  bone2Name=spine_05
  swing1LimitDegrees=45
  swing1Motion=Free
  swing2LimitDegrees=45
  swing2Motion=Free
  twistLimitDegrees=45
  twistMotion=Free
 }
 {bone1Name=dyn_necklace_06_cr, bone2Name=dyn_necklace_05_cr, swing1LimitDegrees=20, swing2LimitDegrees=10, twistLimitDegrees=5}
 {bone1Name=dyn_necklace_03_lf, bone2Name=base_necklace_03_lf, swing1LimitDegrees=2, swing2LimitDegrees=5, twistLimitDegrees=1}
 {bone1Name=dyn_necklace_04_lf, bone2Name=dyn_necklace_03_lf, swing1LimitDegrees=2, swing2LimitDegrees=5, twistLimitDegrees=1}
 {bone1Name=dyn_necklace_03_rt, bone2Name=base_necklace_03_rt, swing1LimitDegrees=2, swing2LimitDegrees=5, twistLimitDegrees=1}
 {bone1Name=dyn_necklace_04_rt, bone2Name=dyn_necklace_03_rt, swing1LimitDegrees=2, swing2LimitDegrees=5, twistLimitDegrees=1}
 {
  bone1Name=dyn_necklace_04_lf
  bone2Name=dyn_necklace_04_cr
  swing1LimitDegrees=45
  swing1Motion=Free
  swing2LimitDegrees=45
  swing2Motion=Free
  twistLimitDegrees=45
  twistMotion=Free
 }
]
