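# Physics asset for the right leg of the RumbleBeat Beach medium male body cosmetic, simulating a hanging "goatgod" attachment.
# A tiny base capsule roots a two-link chain of a capsule and a flat box, joined by constraints with tight swing and twist limits.
# An unconstrained right-thigh capsule sits alongside the chain, apparently as a collision shape for the dangling piece.
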PhysicsAsset /BRCosmetics/Characters/Player/Male/Medium/Bodies/M_MED_RumbleBeat_Beach/Meshes/M_MED_RumbleBeat_Beach_Physics_RLeg
bodies {
 base_goatgod [{shapeType=Capsule, length=0.505, radius=0.505, shapeName=base_goatgod_capsule, rotation=<rot 0 -90 0>}]
 dyn_goatgod_01 [{shapeType=Capsule, length=6, radius=3, shapeName=dyn_goatgod_01_capsule, center=<2 -1.5 0>, rotation=<rot 90 -5 0>}]
 dyn_goatgod_02 [{shapeType=Box, extentX=14, extentY=2, extentZ=14, shapeName=dyn_goatgod_02_box, center=<4.3042 0.924 -0.0968>, rotation=<rot 0 8 10>}]
 thigh_r [{shapeType=Capsule, length=11, radius=9, shapeName=thigh_r_capsule, center=<9.1251 -0.3496 0>, rotation=<rot 90 -5.0002 0>}]
}
constraints [
 {bone1Name=dyn_goatgod_02, bone2Name=dyn_goatgod_01, swing1LimitDegrees=30, swing2LimitDegrees=10, twistLimitDegrees=5}
 {bone1Name=dyn_goatgod_01, bone2Name=base_goatgod, swing1LimitDegrees=20, swing2LimitDegrees=10, twistLimitDegrees=1}
]
