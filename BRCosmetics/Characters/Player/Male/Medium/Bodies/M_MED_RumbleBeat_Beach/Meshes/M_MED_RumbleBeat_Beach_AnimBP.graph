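# Post-process style AnimBP for the RumbleBeat Beach male body, adding secondary motion to the pose copied from the attached parent mesh.
# A Control Rig pass feeds bicep, tricep and pec Anim Dynamics chains, selected by front-end state and gated by bPbBlendEnabled.
# Clavicle helper offsets, RigidBody sims for right leg and necklace, then corrective bone offsets follow.
# Update event computes gravity override, rigid-body settings and crouch, front-end and Takeout emote flags.
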AnimBlueprint /BRCosmetics/Characters/Player/Male/Medium/Bodies/M_MED_RumbleBeat_Beach/Meshes/M_MED_RumbleBeat_Beach_AnimBP
target_skeleton=&/M_MED_RumbleBeat_Beach_Skeleton
graphs [
 AnimGraph {nodes [
   Output Pose(Result=Use cached pose 'Dynamics_CACHE'()) {layerGroup=None}
   n1 = Control Rig(Source=Copy Pose From Mesh() {bUseAttachedParent=true}) {controlRigAssetReference {blueprintRigClass=&/M_MED_RumbleBeat_Beach_CtrlRig.M_MED_RumbleBeat_Beach_CtrlRig_C}}
   Save cached pose 'ControlRig_CACHE'(Pose=@n1) {cacheName=ControlRig_CACHE}
   n2 = Use cached pose 'ControlRig_CACHE'()
   Save cached pose 'PreDynamics_CACHE'(Pose=Component To Local(ComponentPose=Local To Component(LocalPose=@n2))) {cacheName=PreDynamics_CACHE}
   n3 = Use cached pose 'PreDynamics_CACHE'()
   n4 = Get bPbBlendEnabled() -> (bPbBlendEnabled:bool)
   n5 = Anim Dynamics - Bone: upperarm_bicep_l(ComponentPose=Local To Component(LocalPose=Use cached pose 'PreDynamics_CACHE'()), Alpha=1, ExternalForce="0,0,0") {
    angularSpringConstant=15000
    bLinearSpring=true
    bOverrideLinearDamping=true
    bUseGravityOverride=true
    bUsePlanarLimit=false
    linearDampingOverride=1
    linearSpringConstant=500
    simulationSpace=RootRelative
    boundBone {boneName=upperarm_bicep_l}
    physicsBodyDefinitions [{
      boundBone {boneName=upperarm_bicep_l}
      boxExtents=<38 38 38>
      constraintSetup {angularLimitsMax=<0 0 0>, angularLimitsMin=<0 0 0>, angularTarget=<1 0 0>, linearAxesMax=<0 0 0.5>, linearAxesMin=<0 0 -0.5>}
      localJointOffset=<-2 0 0>
     }]
    relativeSpaceBone {boneName=upperarm_l}
    simSpaceSettings {maxAngularAcceleration=40, maxAngularVelocity=770, simSpaceAngularAlpha=1}
   }
   n6 = Anim Dynamics - Bone: upperarm_bicep_l(ComponentPose=@n5, Alpha=1, ExternalForce="0,0,0") {
    angularDampingOverride=0.9999
    angularSpringConstant=15000
    bAngularSpring=true
    bOverrideAngularDamping=true
    bUseGravityOverride=true
    bUsePlanarLimit=false
    simulationSpace=RootRelative
    boundBone {boneName=upperarm_bicep_l}
    gravityOverride {z=-980}
    physicsBodyDefinitions [{
      boundBone {boneName=upperarm_bicep_l}
      boxExtents=<24 38 24>
      constraintSetup {angularLimitsMax=<1 10 1>, angularLimitsMin=<-1 -10 -1>, angularTarget=<1 0 0>}
      localJointOffset=<-5 0 0>
     }]
    relativeSpaceBone {boneName=upperarm_l}
    simSpaceSettings {maxAngularAcceleration=10, maxAngularVelocity=300, simSpaceAngularAlpha=1}
   }
   n7 = Anim Dynamics - Bone: upperarm_bicep_r_base(ComponentPose=@n6, Alpha=1, ExternalForce="0,0,0") {
    angularSpringConstant=15000
    bLinearSpring=true
    bOverrideLinearDamping=true
    bUseGravityOverride=true
    bUsePlanarLimit=false
    linearDampingOverride=1
    linearSpringConstant=500
    simulationSpace=RootRelative
    boundBone {boneName=upperarm_bicep_r_base}
    physicsBodyDefinitions [{
      boundBone {boneName=upperarm_bicep_r_base}
      boxExtents=<38 38 38>
      constraintSetup {angularLimitsMax=<0 0 0>, angularLimitsMin=<0 0 0>, angularTarget=<1 0 0>, linearAxesMax=<0 0 0.5>, linearAxesMin=<0 0 -0.5>}
      localJointOffset=<2 0 0>
     }]
    relativeSpaceBone {boneName=upperarm_r}
    simSpaceSettings {maxAngularAcceleration=40, maxAngularVelocity=770, simSpaceAngularAlpha=1}
   }
   n8 = Anim Dynamics - Bone: upperarm_bicep_r_base(ComponentPose=@n7, Alpha=1, ExternalForce="0,0,0") {
    angularDampingOverride=0.9999
    angularSpringConstant=15000
    bAngularSpring=true
    bOverrideAngularDamping=true
    bUseGravityOverride=true
    bUsePlanarLimit=false
    simulationSpace=RootRelative
    boundBone {boneName=upperarm_bicep_r_base}
    gravityOverride {z=-980}
    physicsBodyDefinitions [{
      boundBone {boneName=upperarm_bicep_r_base}
      boxExtents=<24 38 24>
      constraintSetup {angularLimitsMax=<1 10 1>, angularLimitsMin=<-1 -10 -1>, angularTarget=<1 0 0>}
      localJointOffset=<-5 0 0>
     }]
    relativeSpaceBone {boneName=upperarm_r}
    simSpaceSettings {maxAngularAcceleration=10, maxAngularVelocity=300, simSpaceAngularAlpha=1}
   }
   n9 = Anim Dynamics - Bone: upperarm_tricep_l(ComponentPose=@n8, Alpha=1, ExternalForce="0,0,0") {
    angularSpringConstant=15000
    bLinearSpring=true
    bOverrideLinearDamping=true
    bUseGravityOverride=true
    bUsePlanarLimit=false
    linearDampingOverride=1
    linearSpringConstant=450
    simulationSpace=RootRelative
    boundBone {boneName=upperarm_tricep_l}
    gravityOverride {z=-100}
    physicsBodyDefinitions [{
      boundBone {boneName=upperarm_tricep_l}
      boxExtents=<48 58 48>
      constraintSetup {angularLimitsMax=<0 0 0>, angularLimitsMin=<0 0 0>, angularTarget=<1 0 0>, linearAxesMax=<0 0.1 0.75>, linearAxesMin=<0 -0.1 -0.75>}
      localJointOffset=<-3 0 0>
     }]
    relativeSpaceBone {boneName=upperarm_l}
    simSpaceSettings {maxAngularAcceleration=30, maxAngularVelocity=300, simSpaceAngularAlpha=1}
   }
   n10 = Anim Dynamics - Bone: upperarm_tricep_l(ComponentPose=@n9, Alpha=1, ExternalForce="0,0,0") {
    angularDampingOverride=0.9999
    angularSpringConstant=50000
    bAngularSpring=true
    bOverrideAngularDamping=true
    bUseGravityOverride=true
    bUsePlanarLimit=false
    simulationSpace=BoneRelative
    boundBone {boneName=upperarm_tricep_l}
    gravityOverride {z=-980}
    physicsBodyDefinitions [{
      boundBone {boneName=upperarm_tricep_l}
      boxExtents=<24 38 24>
      constraintSetup {angularLimitsMax=<1 10 1>, angularLimitsMin=<-1 -10 -1>, angularTarget=<1 0 0>}
      localJointOffset=<5 0 0>
     }]
    relativeSpaceBone {boneName=upperarm_l}
    simSpaceSettings {maxAngularAcceleration=10, maxAngularVelocity=300, simSpaceAngularAlpha=1}
   }
   n11 = Anim Dynamics - Bone: upperarm_tricep_r_base(ComponentPose=@n10, Alpha=1, ExternalForce="0,0,0") {
    angularSpringConstant=15000
    bLinearSpring=true
    bOverrideLinearDamping=true
    bUseGravityOverride=true
    bUsePlanarLimit=false
    linearDampingOverride=1
    linearSpringConstant=450
    simulationSpace=RootRelative
    boundBone {boneName=upperarm_tricep_r_base}
    gravityOverride {z=-100}
    physicsBodyDefinitions [{
      boundBone {boneName=upperarm_tricep_r_base}
      boxExtents=<48 58 48>
      constraintSetup {angularLimitsMax=<0 0 0>, angularLimitsMin=<0 0 0>, angularTarget=<1 0 0>, linearAxesMax=<0 0.1 0.75>, linearAxesMin=<0 -0.1 -0.75>}
      localJointOffset=<3 0 0>
     }]
    relativeSpaceBone {boneName=upperarm_r}
    simSpaceSettings {maxAngularAcceleration=30, maxAngularVelocity=300, simSpaceAngularAlpha=1}
   }
   n12 = Anim Dynamics - Bone: upperarm_tricep_r_base(ComponentPose=@n11, Alpha=1, ExternalForce="0,0,0") {
    angularDampingOverride=0.9999
    angularSpringConstant=50000
    bAngularSpring=true
    bOverrideAngularDamping=true
    bUseGravityOverride=true
    bUsePlanarLimit=false
    simulationSpace=BoneRelative
    boundBone {boneName=upperarm_tricep_r_base}
    gravityOverride {z=-980}
    physicsBodyDefinitions [{
      boundBone {boneName=upperarm_tricep_r_base}
      boxExtents=<24 38 24>
      constraintSetup {angularLimitsMax=<1 10 1>, angularLimitsMin=<-1 -10 -1>, angularTarget=<1 0 0>}
      localJointOffset=<5 0 0>
     }]
    relativeSpaceBone {boneName=upperarm_r}
    simSpaceSettings {maxAngularAcceleration=10, maxAngularVelocity=300, simSpaceAngularAlpha=1}
   }
   n13 = Anim Dynamics - Bone: deform_pec_l(ComponentPose=@n12, Alpha=1, ExternalForce="0,0,0") {
    angularSpringConstant=15000
    bLinearSpring=true
    bOverrideLinearDamping=true
    bUseGravityOverride=true
    bUsePlanarLimit=false
    linearDampingOverride=1
    linearSpringConstant=400
    simulationSpace=RootRelative
    boundBone {boneName=deform_pec_l}
    gravityOverride {z=-200}
    physicsBodyDefinitions [{
      boundBone {boneName=deform_pec_l}
      boxExtents=<50 55 60>
      constraintSetup {angularLimitsMax=<0 0 0>, angularLimitsMin=<0 0 0>, angularTarget=<1 0 0>, linearAxesMax=<0.5 0 1.5>, linearAxesMin=<-0.5 0 -1.5>}
      localJointOffset=<-4 0 0>
     }]
    relativeSpaceBone {boneName=spine_04}
    simSpaceSettings {maxAngularAcceleration=10, maxAngularVelocity=770, simSpaceAngularAlpha=1}
   }
   n14 = Anim Dynamics - Bone: deform_pec_r(ComponentPose=@n13, Alpha=1, ExternalForce="0,0,0") {
    angularSpringConstant=15000
    bLinearSpring=true
    bOverrideLinearDamping=true
    bUseGravityOverride=true
    bUsePlanarLimit=false
    linearDampingOverride=1
    linearSpringConstant=400
    simulationSpace=RootRelative
    boundBone {boneName=deform_pec_r}
    gravityOverride {z=-200}
    physicsBodyDefinitions [{
      boundBone {boneName=deform_pec_r}
      boxExtents=<50 55 60>
      constraintSetup {angularLimitsMax=<0 0 0>, angularLimitsMin=<0 0 0>, angularTarget=<1 0 0>, linearAxesMax=<0.5 0 1.5>, linearAxesMin=<-0.5 0 -1.5>}
      localJointOffset=<4 0 0>
     }]
    relativeSpaceBone {boneName=spine_04}
    simSpaceSettings {maxAngularAcceleration=10, maxAngularVelocity=770, simSpaceAngularAlpha=1}
   }
   n15 = Anim Dynamics - Bone: upperarm_bicep_l!(ComponentPose=Local To Component(LocalPose=Use cached pose 'PreDynamics_CACHE'()), ExternalForce="0,0,0") -> (Pose) {
    alpha=0.8
    angularSpringConstant=15000
    bLinearSpring=true
    bOverrideLinearDamping=true
    bUseGravityOverride=true
    bUsePlanarLimit=false
    linearDampingOverride=1
    linearSpringConstant=500
    simulationSpace=RootRelative
    boundBone {boneName=upperarm_bicep_l}
    physicsBodyDefinitions [{
      boundBone {boneName=upperarm_bicep_l}
      boxExtents=<5 5 5>
      constraintSetup {angularLimitsMax=<0 0 0>, angularLimitsMin=<0 0 0>, angularTarget=<1 0 0>, linearAxesMax=<0 1.5 1.5>, linearAxesMin=<0 -1.5 -1.5>}
      localJointOffset=<2 0 0>
     }]
    relativeSpaceBone {boneName=pelvis}
    simSpaceSettings {maxAngularAcceleration=40, maxAngularVelocity=770, simSpaceAngularAlpha=1}
   }
   n16 = Anim Dynamics - Bone: upperarm_bicep_l(ComponentPose=@n15, ExternalForce="0,0,0") {
    alpha=0.8
    angularDampingOverride=0.9999
    angularSpringConstant=15000
    bAngularSpring=true
    bOverrideAngularDamping=true
    bUseGravityOverride=true
    bUsePlanarLimit=false
    simulationSpace=RootRelative
    boundBone {boneName=upperarm_bicep_l}
    componentAppliedLinearAccClamp=<1000 1000 1000>
    gravityOverride {z=-980}
    physicsBodyDefinitions [{
      boundBone {boneName=upperarm_bicep_l}
      boxExtents=<24 38 24>
      constraintSetup {angularLimitsMax=<1 10 1>, angularLimitsMin=<-1 -10 -1>, angularTarget=<1 0 0>}
      localJointOffset=<5 0 0>
     }]
    relativeSpaceBone {boneName=pelvis}
    simSpaceSettings {maxAngularAcceleration=10, maxAngularVelocity=300, simSpaceAngularAlpha=0.4}
   }
   n17 = Anim Dynamics - Bone: upperarm_bicep_r_base!(ComponentPose=@n16, ExternalForce="0,0,0") -> (Pose) {
    alpha=0.8
    angularSpringConstant=15000
    bLinearSpring=true
    bOverrideLinearDamping=true
    bUseGravityOverride=true
    bUsePlanarLimit=false
    linearDampingOverride=1
    linearSpringConstant=500
    simulationSpace=RootRelative
    boundBone {boneName=upperarm_bicep_r_base}
    physicsBodyDefinitions [{
      boundBone {boneName=upperarm_bicep_r_base}
      boxExtents=<5 5 5>
      constraintSetup {angularLimitsMax=<0 0 0>, angularLimitsMin=<0 0 0>, angularTarget=<1 0 0>, linearAxesMax=<0 1.5 1.5>, linearAxesMin=<0 -1.5 -1.5>}
      localJointOffset=<2 0 0>
     }]
    relativeSpaceBone {boneName=pelvis}
    simSpaceSettings {maxAngularAcceleration=40, maxAngularVelocity=770, simSpaceAngularAlpha=1}
   }
   n18 = Anim Dynamics - Bone: upperarm_bicep_r_base(ComponentPose=@n17, ExternalForce="0,0,0") {
    alpha=0.8
    angularDampingOverride=0.9999
    angularSpringConstant=15000
    bAngularSpring=true
    bOverrideAngularDamping=true
    bUseGravityOverride=true
    bUsePlanarLimit=false
    simulationSpace=RootRelative
    boundBone {boneName=upperarm_bicep_r_base}
    componentAppliedLinearAccClamp=<1000 1000 1000>
    gravityOverride {z=-980}
    physicsBodyDefinitions [{
      boundBone {boneName=upperarm_bicep_r_base}
      boxExtents=<24 38 24>
      constraintSetup {angularLimitsMax=<1 10 1>, angularLimitsMin=<-1 -10 -1>, angularTarget=<1 0 0>}
      localJointOffset=<5 0 0>
     }]
    relativeSpaceBone {boneName=pelvis}
    simSpaceSettings {maxAngularAcceleration=10, maxAngularVelocity=300, simSpaceAngularAlpha=0.4}
   }
   n19 = Anim Dynamics - Bone: upperarm_tricep_l!(ComponentPose=@n18, ExternalForce="0,0,0") -> (Pose) {
    alpha=0.8
    angularSpringConstant=15000
    bLinearSpring=true
    bOverrideLinearDamping=true
    bUseGravityOverride=true
    bUsePlanarLimit=false
    linearDampingOverride=1
    linearSpringConstant=450
    simulationSpace=RootRelative
    boundBone {boneName=upperarm_tricep_l}
    gravityOverride {z=-100}
    physicsBodyDefinitions [{
      boundBone {boneName=upperarm_tricep_l}
      boxExtents=<5 5 5>
      constraintSetup {angularLimitsMax=<0 0 0>, angularLimitsMin=<0 0 0>, angularTarget=<1 0 0>, linearAxesMax=<0 0.5 1.5>, linearAxesMin=<0 -0.5 -1.5>}
      localJointOffset=<-3 0 0>
     }]
    relativeSpaceBone {boneName=pelvis}
    simSpaceSettings {maxAngularAcceleration=40, maxAngularVelocity=770, simSpaceAngularAlpha=1}
   }
   n20 = Anim Dynamics - Bone: upperarm_tricep_l(ComponentPose=@n19, ExternalForce="0,0,0") {
    alpha=0.8
    angularDampingOverride=0.9999
    angularSpringConstant=50000
    bAngularSpring=true
    bOverrideAngularDamping=true
    bUseGravityOverride=true
    bUsePlanarLimit=false
    simulationSpace=RootRelative
    boundBone {boneName=upperarm_tricep_l}
    componentAppliedLinearAccClamp=<1000 1000 1000>
    gravityOverride {z=-980}
    physicsBodyDefinitions [{
      boundBone {boneName=upperarm_tricep_l}
      boxExtents=<24 38 24>
      constraintSetup {angularLimitsMax=<1 10 1>, angularLimitsMin=<-1 -10 -1>, angularTarget=<1 0 0>}
      localJointOffset=<-5 0 0>
     }]
    relativeSpaceBone {boneName=pelvis}
    simSpaceSettings {maxAngularAcceleration=10, maxAngularVelocity=300, simSpaceAngularAlpha=0.4}
   }
   n21 = Anim Dynamics - Bone: upperarm_tricep_r_base!(ComponentPose=@n20, ExternalForce="0,0,0") -> (Pose) {
    alpha=0.8
    angularSpringConstant=15000
    bLinearSpring=true
    bOverrideLinearDamping=true
    bUseGravityOverride=true
    bUsePlanarLimit=false
    linearDampingOverride=1
    linearSpringConstant=450
    simulationSpace=RootRelative
    boundBone {boneName=upperarm_tricep_r_base}
    gravityOverride {z=-100}
    physicsBodyDefinitions [{
      boundBone {boneName=upperarm_tricep_r_base}
      boxExtents=<5 5 5>
      constraintSetup {angularLimitsMax=<0 0 0>, angularLimitsMin=<0 0 0>, angularTarget=<1 0 0>, linearAxesMax=<0 0.5 1.5>, linearAxesMin=<0 -0.5 -1.5>}
      localJointOffset=<-3 0 0>
     }]
    relativeSpaceBone {boneName=pelvis}
    simSpaceSettings {maxAngularAcceleration=40, maxAngularVelocity=770, simSpaceAngularAlpha=1}
   }
   n22 = Anim Dynamics - Bone: upperarm_tricep_r_base(ComponentPose=@n21, ExternalForce="0,0,0") {
    alpha=0.8
    angularDampingOverride=0.9999
    angularSpringConstant=50000
    bAngularSpring=true
    bOverrideAngularDamping=true
    bUseGravityOverride=true
    bUsePlanarLimit=false
    simulationSpace=RootRelative
    boundBone {boneName=upperarm_tricep_r_base}
    componentAppliedLinearAccClamp=<1000 1000 1000>
    gravityOverride {z=-980}
    physicsBodyDefinitions [{
      boundBone {boneName=upperarm_tricep_r_base}
      boxExtents=<24 38 24>
      constraintSetup {angularLimitsMax=<1 10 1>, angularLimitsMin=<-1 -10 -1>, angularTarget=<1 0 0>}
      localJointOffset=<-5 0 0>
     }]
    relativeSpaceBone {boneName=pelvis}
    simSpaceSettings {maxAngularAcceleration=10, maxAngularVelocity=300, simSpaceAngularAlpha=0.4}
   }
   n23 = Anim Dynamics - Bone: deform_pec_l!(ComponentPose=@n22, ExternalForce="0,0,0") -> (Pose) {
    alpha=0.8
    angularSpringConstant=15000
    bLinearSpring=true
    bOverrideLinearDamping=true
    bUseGravityOverride=true
    bUsePlanarLimit=false
    linearDampingOverride=1
    linearSpringConstant=400
    simulationSpace=RootRelative
    boundBone {boneName=deform_pec_l}
    gravityOverride {z=-200}
    physicsBodyDefinitions [{
      boundBone {boneName=deform_pec_l}
      boxExtents=<50 55 60>
      constraintSetup {angularLimitsMax=<0 0 0>, angularLimitsMin=<0 0 0>, angularTarget=<1 0 0>, linearAxesMax=<0.5 0 1.25>, linearAxesMin=<-0.5 0 -1.25>}
      localJointOffset=<-4 0 0>
     }]
    relativeSpaceBone {boneName=pelvis}
    simSpaceSettings {maxAngularAcceleration=40, maxAngularVelocity=770, simSpaceAngularAlpha=1}
   }
   n24 = Anim Dynamics - Bone: deform_pec_r!(ComponentPose=@n23, ExternalForce="0,0,0") -> (Pose) {
    alpha=0.8
    angularSpringConstant=15000
    bLinearSpring=true
    bOverrideLinearDamping=true
    bUseGravityOverride=true
    bUsePlanarLimit=false
    linearDampingOverride=1
    linearSpringConstant=400
    simulationSpace=RootRelative
    boundBone {boneName=deform_pec_r}
    gravityOverride {z=-200}
    physicsBodyDefinitions [{
      boundBone {boneName=deform_pec_r}
      boxExtents=<50 55 60>
      constraintSetup {angularLimitsMax=<0 0 0>, angularLimitsMin=<0 0 0>, angularTarget=<1 0 0>, linearAxesMax=<0.5 0 1.25>, linearAxesMin=<-0.5 0 -1.25>}
      localJointOffset=<4 0 0>
     }]
    relativeSpaceBone {boneName=pelvis}
    simSpaceSettings {maxAngularAcceleration=40, maxAngularVelocity=770, simSpaceAngularAlpha=1}
   }
   n25 = Blend Poses by bool(
    BlendPose_0=Component To Local(ComponentPose=@n24)
    BlendPose_1=Component To Local(ComponentPose=@n14)
    BlendTime_0=0
    BlendTime_1=0
    bActiveValue=Get bIsInFrontEnd()
   )
   n26 = "Transform (Modify) Bone - Bone: clavicle_l_physics_help"!(
    ComponentPose=Local To Component(LocalPose=Blend Poses by bool(BlendPose_0=@n25, BlendPose_1=@n3, BlendTime_0=0, BlendTime_1=0, bActiveValue=@n4))
    bAlphaBoolEnabled:bool=True
    Translation="0,-8,8"
    Rotation="0,0,0"
    Scale="1,1,1"
   ) -> (Pose) {
    alphaInputType=Bool
    rotationMode=BMM_Additive
    rotationSpace=BCS_BoneSpace
    translationMode=BMM_Additive
    translationSpace=BCS_BoneSpace
    alphaBoolBlend {blendInTime=0.5, blendOption=QuadraticInOut, blendOutTime=1}
    boneToModify {boneName=clavicle_l_physics_help}
    translation {y=-8, z=8}
   }
   n27 = "Transform (Modify) Bone - Bone: clavicle_r_physics_help"!(ComponentPose=@n26, bAlphaBoolEnabled:bool=True, Translation="0,8,-8", Rotation="0,0,0", Scale="1,1,1") -> (Pose) {
    alphaInputType=Bool
    rotationMode=BMM_Additive
    rotationSpace=BCS_BoneSpace
    translationMode=BMM_Additive
    translationSpace=BCS_BoneSpace
    alphaBoolBlend {blendInTime=0.5, blendOption=QuadraticInOut, blendOutTime=1}
    boneToModify {boneName=clavicle_r_physics_help}
    translation {y=8, z=-8}
   }
   n28 = "Transform (Modify) Bone - Bone: clavicle_l_physics_help"(ComponentPose=@n27, bAlphaBoolEnabled=True, Translation="0,-50,0", Rotation="0,0,0", Scale="1,1,1") {
    alphaInputType=Bool
    rotationMode=BMM_Additive
    rotationSpace=BCS_BoneSpace
    translationMode=BMM_Additive
    translationSpace=BCS_BoneSpace
    alphaBoolBlend {blendOption=QuadraticInOut}
    boneToModify {boneName=clavicle_l_physics_help}
    translation {y=-50}
   }
   n29 = "Transform (Modify) Bone - Bone: clavicle_r_physics_help"(ComponentPose=@n28, bAlphaBoolEnabled=True, Translation="0,50,0", Rotation="0,0,0", Scale="1,1,1") {
    alphaInputType=Bool
    rotationMode=BMM_Additive
    rotationSpace=BCS_BoneSpace
    translationMode=BMM_Additive
    translationSpace=BCS_BoneSpace
    alphaBoolBlend {blendOption=QuadraticInOut}
    boneToModify {boneName=clavicle_r_physics_help}
    translation {y=50}
   }
   n30 = RigidBody!(
    ComponentPose=@n29
    Alpha
    OverrideWorldGravity:Vector="0,0,0"
    ExternalForce
    SimSpaceSettings:Sim Space Settings Structure=(WorldAlpha=1,VelocityScaleZ=0.75,DampingAlpha=0,MaxLinearVelocity=400,MaxAngularVelocity=4,MaxLinearAcceleration=1000,MaxAngularAcceleration=9,ExternalLinearDragV=(X=1.1,Y=1.1,Z=1.1),ExternalLinearVelocity=(X=0,Y=0,Z=0),ExternalAngularVelocity=(X=0,Y=0,Z=0))
   ) -> (Pose) {
    bClampLinearTranslationLimitToRefPose=true
    bOverrideWorldGravity=true
    overridePhysicsAsset=&/M_MED_RumbleBeat_Beach_Physics_RLeg
    simulationSpace=BaseBoneSpace
    baseBoneRef {boneName=thigh_r}
    simSpaceSettings {
     dampingAlpha=0
     maxAngularAcceleration=9
     maxAngularVelocity=4
     maxLinearAcceleration=1000
     maxLinearVelocity=400
     velocityScaleZ=0.75
     worldAlpha=1
     externalLinearDragV=<1.1 1.1 1.1>
    }
   }
   n31 = RigidBody(
    ComponentPose=@n30
    OverrideWorldGravity="0,0,0"
    SimSpaceSettings=(WorldAlpha=1,VelocityScaleZ=0.75,DampingAlpha=0,MaxLinearVelocity=400,MaxAngularVelocity=4,MaxLinearAcceleration=1000,MaxAngularAcceleration=9,ExternalLinearDragV=(X=1.1,Y=1.1,Z=1.1),ExternalLinearVelocity=(X=0,Y=0,Z=0),ExternalAngularVelocity=(X=0,Y=0,Z=0))
   ) {
    bClampLinearTranslationLimitToRefPose=true
    bOverrideWorldGravity=true
    overridePhysicsAsset=&/M_MED_RumbleBeat_Beach_Physics_Necklace
    simulationSpace=BaseBoneSpace
    baseBoneRef {boneName=spine_05}
    simSpaceSettings {
     dampingAlpha=0
     maxAngularAcceleration=9
     maxAngularVelocity=4
     maxLinearAcceleration=1000
     maxLinearVelocity=400
     velocityScaleZ=0.75
     worldAlpha=1
     externalLinearDragV=<1.1 1.1 1.1>
    }
   }
   n32 = "Transform (Modify) Bone - Bone: base_necklace_03_lf"!(ComponentPose=@n31, bAlphaBoolEnabled:bool=True, Translation="1,2,1", Rotation="-20,0,0", Scale="1,1,1") -> (Pose) {
    alphaInputType=Bool
    rotationMode=BMM_Additive
    rotationSpace=BCS_BoneSpace
    translationMode=BMM_Additive
    translationSpace=BCS_BoneSpace
    alphaBoolBlend {blendInTime=0.5, blendOption=QuadraticInOut, blendOutTime=1}
    boneToModify {boneName=base_necklace_03_lf}
    rotation {pitch=-20}
    translation=<1 2 1>
   }
   n33 = "Transform (Modify) Bone - Bone: dyn_necklace_04_cr"!(ComponentPose=@n32, bAlphaBoolEnabled:bool=True, Translation="0,0,0", Rotation="-10,0,0", Scale="1,1,1") -> (Pose) {
    alphaInputType=Bool
    rotationMode=BMM_Additive
    rotationSpace=BCS_BoneSpace
    translationMode=BMM_Additive
    translationSpace=BCS_BoneSpace
    alphaBoolBlend {blendInTime=0.5, blendOption=QuadraticInOut, blendOutTime=1}
    boneToModify {boneName=dyn_necklace_04_cr}
    rotation {pitch=-10}
   }
   n34 = "Transform (Modify) Bone - Bone: base_goatgod"!(ComponentPose=@n33, bAlphaBoolEnabled:bool=True, Translation="0,0,0", Rotation="0,8,0", Scale="1,1,1") -> (Pose) {
    alphaInputType=Bool
    rotationMode=BMM_Additive
    rotationSpace=BCS_BoneSpace
    translationMode=BMM_Additive
    translationSpace=BCS_BoneSpace
    alphaBoolBlend {blendInTime=0.5, blendOption=QuadraticInOut, blendOutTime=1}
    boneToModify {boneName=base_goatgod}
    rotation {yaw=8}
   }
   n35 = "Transform (Modify) Bone - Bone: base_goatgod"(ComponentPose=@n34, bAlphaBoolEnabled=True, Translation="0,-3,0", Rotation="15,10,-15", Scale="1,1,1") {
    alphaInputType=Bool
    rotationMode=BMM_Additive
    rotationSpace=BCS_BoneSpace
    translationMode=BMM_Additive
    translationSpace=BCS_BoneSpace
    alphaBoolBlend {blendInTime=0.5, blendOption=QuadraticInOut, blendOutTime=1}
    boneToModify {boneName=base_goatgod}
    rotation=<rot 15 10 -15>
    translation {y=-3}
   }
   Save cached pose 'Dynamics_CACHE'(Pose=Component To Local(ComponentPose=@n35)) {cacheName=Dynamics_CACHE}
  ]}
 EventGraph {nodes [
   Event BlueprintUpdateAnimation()
   > n1 = Sequence!(execute) -> (then_0, then_1, then_2:Exec)
   n2 = Get bIsSliding()
   Set JustCrouching(execute=@n1.then_1, JustCrouching:bool=AND Boolean(A=Get bIsCrouching(), B=NOT Boolean(A=@n2))) -> (Output_Get:bool)
   n3 = Get bIsPlayingEmote()
   > Set FENoEmote(FENoEmote:bool=AND Boolean(A=Get bIsInFrontEnd(), B=NOT Boolean(A=@n3))) -> (Output_Get:bool)
   n4 = "Gravity Override State Function (Parameter Asset) (Nativized)"(
    execute=@n1.then_0
    CharacterPartAnimInstance=Self-Reference()
    GameTimeInSeconds=GetGameTimeInSeconds()
    Parameters=&/RumbleBeat_Beach_Parameters
   )
   > Set GravityOverride(GravityOverride=@n4)
   > Set RBAN Settings Out(RBAN Settings Out=@n4.OutRigidBodyAnimNodeParameters)
   n5 = GetAnimInstance(self=GetOwnerSkeletalMeshComponent())
   Is Valid(exec=@n1.then_2, InputObject=@n5)
   > Set Emote_TakeOut(Emote_TakeOut:bool=Montage_IsActive!(
     self:Anim Instance Object Reference=@n5
     Montage:Anim Montage Object Reference=/BRCosmetics/Animation/Game/MainPlayer/Emotes/Takeout/Emote_Takeout_CMM_M
    ) -> (ReturnValue:bool)) -> (Output_Get:bool)
  ]}
]
variables [
 Emote_TakeOut {type=boolean}
 FENoEmote {type=boolean}
 GravityOverride {type=Vector}
 JustCrouching {type=boolean}
 RBAN Settings Out {type=FortRigidBodyAnimNodeParameters}
]
