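# Physics asset of collision volumes for the M_MED_RumbleBeat_Beach player body, named as a cape collider.
# Capsules cover the pelvis and five spine bones, a sphere covers the head, and each clavicle has a mirrored pair of overlapping capsules.
# Arm, leg and foot bodies exist but hold no shapes, so only torso, shoulders and head collide.
# It has no constraints between bodies.
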PhysicsAsset /BRCosmetics/Characters/Player/Male/Medium/Bodies/M_MED_RumbleBeat_Beach/Meshes/M_MED_RumbleBeat_Beach_CapeCollider
bodies {
 calf_l []
 calf_r []
 clavicle_l [
  {shapeType=Capsule, length=8.1142, radius=7.7926, center=<7.4828 -2.1782 1.4896>, rotation=<rot 70.9474 0.2671 -178.6668>}
  {shapeType=Capsule, length=6.7451, radius=9.1617, center=<5.9718 2.2634 -3.1429>, rotation=<rot 70.9474 0.2671 -178.6668>}
 ]
 clavicle_r [
  {shapeType=Capsule, length=8.1142, radius=7.7926, center=<-7.4828 2.1782 -1.4896>, rotation=<rot 70.9474 0.2671 1.3332>}
  {shapeType=Capsule, length=6.7451, radius=9.1617, center=<-5.9718 -2.2634 3.1429>, rotation=<rot 70.9474 0.2671 1.3332>}
 ]
 foot_l []
 foot_r []
 head [{shapeType=Sphere, radius=17.3125, center=<11.028 9.0474 0>}]
 lowerarm_l []
 lowerarm_r []
 pelvis [{shapeType=Capsule, length=14, radius=15.5, center=<-1 -1.0272 0>}]
 spine_01 [{shapeType=Capsule, length=16, radius=14.5, center=<0.8067 -1.3326 0>}]
 spine_02 [{shapeType=Capsule, length=9, radius=13.6322, center=<-1.711 -0.5141 0.0001>}]
 spine_03 [{shapeType=Capsule, length=18, radius=14.5, center=<0.3964 -1.0083 0.0001>, rotation=<rot 0 4.3792 0>}]
 spine_04 [{shapeType=Capsule, length=20.9094, radius=15.5, center=<2.766 -0.2909 0.0002>}]
 spine_05 [{shapeType=Capsule, length=20.281, radius=15, center=<2.0053 0.1808 0>}]
 thigh_l []
 thigh_r []
 upperarm_l []
 upperarm_r []
}
constraints []
