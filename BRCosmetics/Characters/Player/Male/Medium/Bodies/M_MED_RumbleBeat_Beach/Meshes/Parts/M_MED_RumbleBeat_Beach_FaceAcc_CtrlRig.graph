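# Control Rig for the RumbleBeat Beach male medium character's FaceAcc part that procedurally poses hair bones, mainly braids and mullet, as the body moves.
# The Aim function layers weighted corrective transforms from head delta rotation and spherical pose readers, then partially aims braid roots at a spine-parented target.
# A left upper-arm pose reader, active only when Target Crouching is true, drives a smoothed, eased weight for the Aim_Up_Fusile curve and another braid corrective.
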ControlRigBlueprint /BRCosmetics/Characters/Player/Male/Medium/Bodies/M_MED_RumbleBeat_Beach/Meshes/Parts/M_MED_RumbleBeat_Beach_FaceAcc_CtrlRig
graphs [
 New Function_ContainedGraph {nodes [
   n1 = RigVMFunctionEntryNode()
   RigVMFunction_MathBoolToInteger = &RigVMFunction_MathBoolToInteger(Value=@n1.onlyTwist)
   RigVMFunction_MathQuaternionSwingTwist = &RigVMFunction_MathQuaternionSwingTwist(
    twistAxis=@n1.twistAxis
    Input=&RigUnit_GetRelativeTransformForItem(
     bChildInitial=False
     bParentInitial=False
     child=(Type=Bone,Name="None")
     parent=(Type=Bone,Name="None")
     child.Name=@n1.child
     parent.Name=@n1.parent
     parent.Type=@n1.parentType
    ).RelativeTransform.Rotation
   )
   RigVMFunction_MathQuaternionToEuler = &RigVMFunction_MathQuaternionToEuler(
    Value=RigVMDispatchNode(
     Values=((),())
     Index=@RigVMFunction_MathBoolToInteger.Result
     Values.0=@RigVMFunction_MathQuaternionSwingTwist.Swing
     Values.1=@RigVMFunction_MathQuaternionSwingTwist.Twist
    ) {template="DISPATCH_RigVMDispatch_SelectInt32(in Index,in Values,out Result)"}.Result
    RotationOrder=@n1.Rotation Order
   )
   > DISPATCH_RigVMDispatch_Print = RigVMDispatchNode(ScreenColor=(R=1,G=1,B=1,A=1), ScreenDuration=1.5, Enabled=@n1.debugRotation, Value=@RigVMFunction_MathQuaternionToEuler.Result) {template="DISPATCH_RigVMDispatch_Print(in Prefix,in Value,in Enabled,in ScreenDuration,in ScreenColor)"}
   Clamp = &RigVMFunction_MathIntClamp(Value=@n1.primAxis)
   DISPATCH_RigVMDispatch_SelectInt32 = RigVMDispatchNode(
    Values=(0,0,0)
    Index=@Clamp.Result
    Values.0=@RigVMFunction_MathQuaternionToEuler.Result.X
    Values.1=@RigVMFunction_MathQuaternionToEuler.Result.Y
    Values.2=@RigVMFunction_MathQuaternionToEuler.Result.Z
   ) {template="DISPATCH_RigVMDispatch_SelectInt32(in Index,in Values,out Result)"}
   Remap = &RigVMFunction_MathDoubleRemap(
    TargetMaximum=1
    TargetMinimum=0
    bClamp=False
    Value=@DISPATCH_RigVMDispatch_SelectInt32.Result
    SourceMaximum=@n1.max
    SourceMinimum=@n1.min
   )
   Multiply = &RigVMFunction_MathDoubleMul(A=&RigVMFunction_MathDoubleClamp(Value=@Remap.Result).Result, B=@n1.premult)
   > RigVMFunctionReturnNode(outputWeight=@Multiply.Result)
  ]}
 New Function_ContainedGraph {nodes [
   RigVMFunctionEntryNode()
   > RigVMFunction_Sequence = RigVMAggregateNode()
   &RigUnit_AimConstraintLocalSpaceOffset(
    Child=(Type=Bone,Name="dyn_braid01_cr")
    Filter=(bX=true,bY=True,bZ=True)
    Parents=((Item=(Type=Bone,Name="hair_aim_tgt"),Weight=1))
    Weight=0.5
    WorldUp=(Target=(X=0,Y=0,Z=1),Kind=Direction,Space=(Type=Bone,Name="head"))
    bMaintainOffset=true
    ExecutePin=@RigVMFunction_Sequence.D
   )
   > &RigUnit_AimConstraintLocalSpaceOffset(
    Child=(Type=Bone,Name="dyn_braid01_lf")
    Filter=(bX=true,bY=True,bZ=True)
    Parents=((Item=(Type=Bone,Name="hair_aim_tgt"),Weight=1))
    Weight=0.5
    WorldUp=(Target=(X=0,Y=0,Z=1),Kind=Direction,Space=(Type=Bone,Name="head"))
    bMaintainOffset=true
   )
   > &RigUnit_AimConstraintLocalSpaceOffset(
    Child=(Type=Bone,Name="dyn_braid01_rt")
    Filter=(bX=false,bY=True,bZ=True)
    Parents=((Item=(Type=Bone,Name="hair_aim_tgt"),Weight=1))
    Weight=0.5
    WorldUp=(Target=(X=0,Y=0,Z=1),Kind=Direction,Space=(Type=Bone,Name="head"))
    bMaintainOffset=true
   )
   RigVMFunctionReturnNode(ExecuteContext=@RigVMFunction_Sequence.E)
   n1 = &RigUnit_SphericalPoseReader(
    ActiveRegionScaleFactors=(PositiveWidth=0.75,NegativeWidth=0.75,PositiveHeight=1,NegativeHeight=1)
    ActiveRegionSize=0.1
    Debug=(bDrawDebug=false,bDraw2D=false,DebugScale=25,DebugSegments=20,DebugThickness=0.25)
    DriverAxis=(X=0,Y=1,Z=0)
    DriverItem=(Type=Bone,Name="head")
    FalloffRegionScaleFactors=(PositiveWidth=1,NegativeWidth=1,PositiveHeight=1,NegativeHeight=2.5)
    FalloffSize=0.2
    OptionalParentItem=(Type=Bone)
    RotationOffset=(X=140,Y=0,Z=-15)
    ExecutePin=@RigVMFunction_Sequence.B
   )
   > &RigUnit_ModifyTransforms(
    ItemToModify="((Item=(Type=Bone,Name=\"base_braid01_rt\"),Transform=(Rotation=(X=-0.3391,Y=-0.0446,Z=0.1227,W=0.9317),Translation=(X=0,Y=0,Z=0),Scale3D=(X=1,Y=1,Z=1))),(Item=(Type=Bone,Name=\"base_braid01_cr\"),Transform=(Rotation=(X=-0.3391,Y=-0.0446,Z=0.1227,W=0.9317),Tra ...(truncated)"
    WeightMinimum=0
    Weight=@n1.OutputParam
   )
   > n2 = &RigUnit_SphericalPoseReader(
    ActiveRegionScaleFactors=(PositiveWidth=0.75,NegativeWidth=0.75,PositiveHeight=1.5,NegativeHeight=1)
    ActiveRegionSize=0.15
    Debug=(bDrawDebug=true,bDraw2D=false,DebugScale=25,DebugSegments=20,DebugThickness=0.25)
    DriverAxis=(X=0,Y=1,Z=0)
    DriverItem=(Type=Bone,Name="head")
    FalloffRegionScaleFactors=(PositiveWidth=1,NegativeWidth=1,PositiveHeight=1,NegativeHeight=2.2)
    FalloffSize=0.21
    FlipHeightScaling=true
    FlipWidthScaling=false
    OptionalParentItem=(Type=Bone)
    RotationOffset=(X=35,Y=0,Z=-20)
   )
   > &RigUnit_ModifyTransforms(
    ItemToModify="((Item=(Type=Bone,Name=\"base_braid01_lf\"),Transform=(Rotation=(X=-0.0046,Y=-0.0521,Z=-0.087,W=0.9948),Translation=(X=0,Y=0,Z=0),Scale3D=(X=1,Y=1,Z=1))),(Item=(Type=Bone,Name=\"base_braid01_cr\"),Transform=(Rotation=(X=-0.0023,Y=-0.0523,Z=-0.0436,W=0.9977),Tr ...(truncated)"
    WeightMinimum=0
    Weight=@n2.OutputParam
   )
   n3 = &RigUnit_SphericalPoseReader(
    ActiveRegionScaleFactors=(PositiveWidth=1,NegativeWidth=1,PositiveHeight=1,NegativeHeight=1)
    ActiveRegionSize=0.018
    Debug=(bDrawDebug=false,bDraw2D=false,bDrawLocalAxes=true,DebugScale=50,DebugSegments=13,DebugThickness=0.25)
    DriverAxis=(X=0,Y=-1,Z=0)
    DriverItem=(Type=Bone,Name="head")
    FalloffRegionScaleFactors=(PositiveWidth=1,NegativeWidth=1,PositiveHeight=1.25,NegativeHeight=1.25)
    FalloffSize=0.1
    FlipHeightScaling=False
    FlipWidthScaling=False
    OptionalParentItem=(Type=Bone,Name="None")
    RotationOffset=(X=-80,Y=0,Z=-32)
    ExecutePin=@RigVMFunction_Sequence.C
   )
   > &RigUnit_ModifyTransforms(
    ItemToModify="((Item=(Type=Bone,Name=\"base_braid01_rt\"),Transform=(Rotation=(X=-0.4216,Y=-0.0295,Z=0.0632,W=0.9041),Translation=(X=-2.5,Y=0,Z=0),Scale3D=(X=0.7,Y=1,Z=1))),(Item=(Type=Bone,Name=\"dyn_braid02_rt\"),Transform=(Rotation=(X=0,Y=0,Z=-0.1736,W=0.9848),Translatio ...(truncated)"
    WeightMinimum=0
    Weight=@n3.OutputParam
   )
   DeltaRotation = RigVMFunctionReferenceNode(
    Rotation Order=ZYX
    child=head
    debugRotation=false
    max=-40
    onlyTwist=false
    parent=head_null
    parentType=Null
    premult=1
    primAxis=2
    twistAxis=(X=1,Y=0,Z=0)
    ExecuteContext=@RigVMFunction_Sequence.A
   )
   > &RigUnit_ModifyTransforms(
    ItemToModify="((Item=(Type=Bone,Name=\"dyn_braid02_cr\"),Transform=(Rotation=(X=0,Y=-0.4617,Z=0,W=0.887),Translation=(X=0,Y=0,Z=0),Scale3D=(X=1,Y=1,Z=1))),(Item=(Type=Bone,Name=\"dyn_braid02_lf\"),Transform=(Rotation=(X=0,Y=-0.342,Z=0,W=0.9397),Translation=(X=0,Y=0,Z=0),Sca ...(truncated)"
    WeightMinimum=0
    Weight=@DeltaRotation.outputWeight
   )
   > &RigUnit_ModifyTransforms(
    ItemToModify="((Item=(Type=Bone,Name=\"base_mullet_rt\"),Transform=(Rotation=(X=-0.0915,Y=0.3415,Z=-0.4085,W=0.8415),Translation=(X=0,Y=0,Z=0),Scale3D=(X=0.3,Y=1,Z=1))),(Item=(Type=Bone,Name=\"base_mullet_cr\"),Transform=(Rotation=(X=0,Y=-0.5736,Z=0,W=0.8192),Translation=(X ...(truncated)"
    WeightMinimum=0
    Weight=@DeltaRotation.outputWeight
   )
   > DeltaRotation_1 = RigVMFunctionReferenceNode(
    Rotation Order=ZYX
    child=head
    debugRotation=false
    max=40
    onlyTwist=false
    parent=head_null
    parentType=Null
    premult=1
    primAxis=2
    twistAxis=(X=1,Y=0,Z=0)
   )
   > &RigUnit_ModifyTransforms(
    ItemToModify="((Item=(Type=Bone,Name=\"base_braid01_rt\"),Transform=(Rotation=(X=0,Y=-0.2164,Z=0,W=0.9763),Translation=(X=0,Y=0,Z=0),Scale3D=(X=1,Y=1,Z=1))),(Item=(Type=Bone,Name=\"base_braid01_lf\"),Transform=(Rotation=(X=0,Y=-0.2164,Z=0,W=0.9763),Translation=(X=0,Y=0,Z=0) ...(truncated)"
    WeightMinimum=0
    Weight=@DeltaRotation_1.outputWeight
   )
  ]}
 RigVMFunctionLibrary {nodes [
   Aim = RigVMCollapseNode(ExecuteContext=())
   DeltaRotation = RigVMCollapseNode(ExecuteContext=(), Rotation Order=ZYX, premult=1, twistAxis=(X=1,Y=0,Z=0))
  ]}
 RigVMFunction_Sequence_1_ContainedGraph {nodes [
   n1 = RigVMFunctionEntryNode()
   RigVMFunction_Sequence = &RigVMFunction_Sequence(ExecuteContext=@n1.ExecuteContext)
   n2 = &RigVMFunction_Sequence(ExecuteContext=@RigVMFunction_Sequence.B)
   n3 = &RigVMFunction_Sequence(ExecuteContext=@n2.B)
   n4 = &RigVMFunction_Sequence(ExecuteContext=@n3.B)
   RigVMFunctionReturnNode(A=@RigVMFunction_Sequence.A, B=@n2.A, C=@n3.A, D=@n4.A, E=@n4.B)
  ]}
 RigVMModel {nodes [
   RigUnit_BeginExecution = &RigUnit_BeginExecution()
   RigVMFunction_Sequence = &RigVMFunction_Sequence(ExecuteContext=@RigUnit_BeginExecution.ExecutePin)
   Aim = RigVMFunctionReferenceNode(ExecuteContext=@RigVMFunction_Sequence.A)
   n1 = &RigUnit_SphericalPoseReader(
    ActiveRegionScaleFactors=(PositiveWidth=1,NegativeWidth=1,PositiveHeight=1,NegativeHeight=1)
    ActiveRegionSize=0.07
    Debug=(bDrawDebug=true,bDraw2D=False,bDrawLocalAxes=False,DebugScale=25,DebugSegments=20,DebugThickness=0.25)
    DriverAxis=(X=1,Y=0,Z=0)
    DriverItem=(Type=Bone,Name="upperarm_l")
    FalloffRegionScaleFactors=(PositiveWidth=1,NegativeWidth=1,PositiveHeight=1,NegativeHeight=1)
    FalloffSize=0.11
    FlipHeightScaling=False
    FlipWidthScaling=False
    OptionalParentItem=(Type=Bone,Name="None")
    RotationOffset=(X=30,Y=-20,Z=-30)
    ExecutePin=@RigVMFunction_Sequence.B
   )
   AlphaInterp = &RigVMFunction_AlphaInterp(
    InterpSpeedDecreasing=3
    InterpSpeedIncreasing=3
    Scale=1
    bClampResult=true
    bMapRange=true
    Value=RigVMDispatchNode(False=0, True=@n1.OutputParam, Condition=RigVMVariableNode() {variable=Target Crouching}.Value) {template="DISPATCH_RigVMDispatch_If(in Condition,in True,in False,out Result)"}.Result
   )
   RigVMFunction_AnimEvalRichCurve = &RigVMFunction_AnimEvalRichCurve(
    Curve=(EditorCurveData=(Keys=((InterpMode=RCIM_Cubic),(InterpMode=RCIM_Cubic,Time=1,Value=1))))
    SourceMinimum=0
    TargetMinimum=0
    Value=@AlphaInterp.Result
   )
   > /Script/ControlRig.RigUnit_SetCurveValue(Curve=Aim_Up_Fusile, Value=@RigVMFunction_AnimEvalRichCurve.Result) {is=struct RigVMUnitNode}
   > &RigUnit_ModifyTransforms(
    ItemToModify="((Item=(Type=Bone,Name=\"base_braid01_rt\"),Transform=(Rotation=(X=0,Y=-0.5,Z=0,W=0.866),Translation=(X=-2.8,Y=0,Z=0),Scale3D=(X=1,Y=1,Z=1))),(Item=(Type=Bone,Name=\"base_braid01_lf\"),Transform=(Rotation=(X=0,Y=-0.3007,Z=0,W=0.9537),Translation=(X=0,Y=0,Z=0), ...(truncated)"
    WeightMinimum=0
    Weight=@RigVMFunction_AnimEvalRichCurve.Result
   )
  ]}
]
hierarchy [
 root {loc=<0 0 0>, rot=<0 0 0 1>}
 pelvis {parent=root, loc=<0 -0.7071 93.1191>, rot=<0.0127 0.707 -0.0127 -0.707>}
 spine_01 {parent=pelvis, loc=<6.7494 -0.6388 0>, rot=<0 0 0.034 -0.9994>}
 spine_02 {parent=spine_01, loc=<7.2518 -0.035 0>, rot=<0 0 0.0238 -0.9997>}
 spine_03 {parent=spine_02, loc=<6.0104 0 0>, rot=<0 0 0.0149 -0.9999>}
 spine_04 {parent=spine_03, loc=<8.1024 0 0>, rot=<0 0 -0.0132 -0.9999>}
 spine_05 {parent=spine_04, loc=<10.0146 0 0>, rot=<0 0 -0.0015 -1>}
 neck_01 {parent=spine_05, loc=<11.3138 0.1529 0>, rot=<0 0 -0.2003 -0.9797>}
 head_null {parent=neck_01, type=Null, loc=<9.9266 -0.5431 0>, rot=<0 0 -0.1649 0.9863>}
 neck_02 {parent=neck_01, loc=<4.5627 0 0>, rot=<0 0 0.0504 -0.9987>}
 head {parent=neck_02, loc=<5.3914 0 0>, rot=<0 0 0.115 -0.9934>}
 base_bangs01_rt {parent=head, loc=<9.5599 8.372 1.3094>, rot=<0.6732 0.3673 0.5185 0.3783>}
 dyn_bangs01_rt {parent=base_bangs01_rt, loc=<0 0 0>, rot=<0 0 0 1>}
 dyn_bangs02_rt {parent=dyn_bangs01_rt, loc=<2.56 0 0>, rot=<0 -0.5855 0 -0.8107>}
 dyn_bangs03_rt {parent=dyn_bangs02_rt, loc=<4.1391 0 0>, rot=<-0.1949 0.0578 0.0067 -0.9791>}
 base_mullet_cr {parent=head, loc=<0.8815 -8.1771 -0.0717>, rot=<0.0414 -0.7059 0.7059 -0.0414>}
 dyn_mullet_cr {parent=base_mullet_cr, loc=<0 0 0>, rot=<0 0 0 1>}
 base_mullet_rt {parent=head, loc=<0.4573 -5.1414 5.7759>, rot=<0.0189 -0.3769 0.926 -0.012>}
 dyn_mullet_rt {parent=base_mullet_rt, loc=<0 0 0>, rot=<0 0 0 1>}
 base_mullet_lf {parent=head, loc=<0.4576 -5.1414 -5.776>, rot=<0.012 -0.926 0.3769 -0.0189>}
 dyn_mullet_lf {parent=base_mullet_lf, loc=<0 0 0>, rot=<0 0 0 1>}
 base_braid01_cr {parent=head, loc=<7.2853 -7.788 -0.1488>, rot=<0.109 -0.6987 0.6987 -0.109>}
 dyn_braid01_cr {parent=base_braid01_cr, loc=<0 0 0>, rot=<0 0 0 1>}
 dyn_braid02_cr {parent=dyn_braid01_cr, loc=<4.9412 0 0>, rot=<0 -0.0961 0 -0.9954>}
 dyn_braid03_cr {parent=dyn_braid02_cr, loc=<3.4031 0 0>, rot=<0 0.0527 0 -0.9986>}
 dyn_braid04_cr {parent=dyn_braid03_cr, loc=<5.1794 0 0>, rot=<0 0 0 1>}
 base_braid01_rt {parent=head, loc=<7.6112 -6.7957 3.4109>, rot=<0.1457 -0.5765 0.7954 -0.1172>}
 dyn_braid01_rt {parent=base_braid01_rt, loc=<0 0 0>, rot=<0 0 0 1>}
 dyn_braid02_rt {parent=dyn_braid01_rt, loc=<4.891 0 0.0002>, rot=<0 -0.1586 0 -0.9873>}
 dyn_braid03_rt {parent=dyn_braid02_rt, loc=<3.8782 0 -0.0001>, rot=<0.1135 0.1398 -0.0161 -0.9835>}
 dyn_braid04_rt {parent=dyn_braid03_rt, loc=<4.3123 0 0.0001>, rot=<0 -0.043 0 -0.9991>}
 base_braid01_lf {parent=head, loc=<7.6113 -6.7957 -3.4109>, rot=<0.1172 -0.7954 0.5765 -0.1457>}
 dyn_braid01_lf {parent=base_braid01_lf, loc=<0 0 0>, rot=<0 0 0 1>}
 dyn_braid02_lf {parent=dyn_braid01_lf, loc=<4.8914 0 0>, rot=<0 -0.1586 0 -0.9873>}
 dyn_braid03_lf {parent=dyn_braid02_lf, loc=<3.8776 0 0>, rot=<-0.1135 0.1398 0.0161 -0.9835>}
 dyn_braid04_lf {parent=dyn_braid03_lf, loc=<4.3125 0 0>, rot=<0 -0.043 0 -0.9991>}
 base_hairTop_cr {parent=head, loc=<13.847 2.8491 0>, rot=<0.5773 -0.4083 0.4083 -0.5773>}
 dyn_hairTop_cr {parent=base_hairTop_cr, loc=<0 0 0>, rot=<0 0 0 1>}
 FACIAL_C_FacialRoot {parent=head, loc=<2.1524 -0.746 -0.0002>, rot=<0.4879 -0.5118 -0.4879 -0.5118>}
 FACIAL_L_EyesackLower {parent=FACIAL_C_FacialRoot, loc=<2.9118 -0.2118 7.6611>, rot=<-0.0209 -0.1396 -0.0222 -0.9897>}
 FACIAL_R_EyesackLower {parent=FACIAL_C_FacialRoot, loc=<-2.9145 -0.2112 7.6482>, rot=<-0.0209 0.1399 0.022 -0.9897>}
 FACIAL_L_CheekInner {parent=FACIAL_C_FacialRoot, loc=<2.0499 0.4855 8.0034>, rot=<-0.1218 -0.0882 -0.1971 -0.9688>}
 FACIAL_R_CheekInner {parent=FACIAL_C_FacialRoot, loc=<-2.0511 0.4858 8.0038>, rot=<-0.1219 0.0884 0.1969 -0.9688>}
 FACIAL_L_CheekOuter {parent=FACIAL_C_FacialRoot, loc=<3.9943 0.5851 6.2057>, rot=<0.0565 -0.3767 0.1008 -0.9191>}
 FACIAL_R_CheekOuter {parent=FACIAL_C_FacialRoot, loc=<-3.9905 0.5997 6.2028>, rot=<0.0564 0.3769 -0.101 -0.919>}
 FACIAL_L_NasolabialBulge {parent=FACIAL_C_FacialRoot, loc=<3.8962 3.9774 6.2086>, rot=<0.0593 -0.3772 0.0242 -0.9239>}
 FACIAL_R_NasolabialBulge {parent=FACIAL_C_FacialRoot, loc=<-3.8929 3.9809 6.2032>, rot=<0.0592 0.3774 -0.0244 -0.9238>}
 FACIAL_L_NasolabialFurrow {parent=FACIAL_C_FacialRoot, loc=<3.0293 2.9574 8.7307>, rot=<-0.0319 -0.1262 -0.3636 -0.9224>}
 FACIAL_R_NasolabialFurrow {parent=FACIAL_C_FacialRoot, loc=<-3.0217 2.966 8.731>, rot=<-0.032 0.1264 0.3633 -0.9225>}
 FACIAL_L_CheekLower {parent=FACIAL_C_FacialRoot, loc=<4.4371 3.7298 5.5137>, rot=<0.0359 -0.4267 0.0583 -0.9018>}
 FACIAL_R_CheekLower {parent=FACIAL_C_FacialRoot, loc=<-4.4341 3.7334 5.5076>, rot=<0.0358 0.4269 -0.0584 -0.9017>}
 FACIAL_C_Nose {parent=FACIAL_C_FacialRoot, loc=<-0.0011 0.2028 10.0455>, rot=<-0.1289 0.0001 -0.0001 -0.9917>}
 FACIAL_C_NoseLower {parent=FACIAL_C_Nose, loc=<0.0013 2.829 -0.3069>, rot=<0.1305 0 0 -0.9914>}
 FACIAL_L_Nostril {parent=FACIAL_C_Nose, loc=<1.3034 2.1236 -0.6103>, rot=<0.0173 0.1305 0.0023 -0.9913>}
 FACIAL_R_Nostril {parent=FACIAL_C_Nose, loc=<-1.304 2.1264 -0.6131>, rot=<0.0173 -0.1305 -0.0023 -0.9913>}
 FACIAL_C_NoseTip {parent=FACIAL_C_Nose, loc=<0.0006 1.9748 0.8582>, rot=<0.0872 0 0 -0.9962>}
 FACIAL_C_MouthUpper {parent=FACIAL_C_FacialRoot, loc=<0.0016 4.4039 7.2814>, rot=<0.0014 0.0001 -0.0001 -1>}
 FACIAL_C_LipUpper {parent=FACIAL_C_MouthUpper, loc=<-0.0009 -0.379 2.8084>, rot=<-0.0872 0 0 -0.9962>}
 FACIAL_L_LipUpper {parent=FACIAL_C_MouthUpper, loc=<1.2612 -0.2621 2.425>, rot=<-0.0778 -0.1601 -0.0379 -0.9833>}
 FACIAL_R_LipUpper {parent=FACIAL_C_MouthUpper, loc=<-1.2594 -0.2656 2.425>, rot=<-0.0778 0.1601 0.0379 -0.9833>}
 FACIAL_L_LipUpperOuter {parent=FACIAL_C_MouthUpper, loc=<2.0727 0.1095 1.7324>, rot=<-0.0984 -0.2293 -0.0983 -0.9634>}
 FACIAL_R_LipUpperOuter {parent=FACIAL_C_MouthUpper, loc=<-2.0721 0.1064 1.7307>, rot=<-0.0984 0.2293 0.0983 -0.9634>}
 FACIAL_L_LipCorner {parent=FACIAL_C_MouthUpper, loc=<2.4545 0.4271 0.9141>, rot=<0 -0.2078 0 -0.9782>}
 FACIAL_R_LipCorner {parent=FACIAL_C_MouthUpper, loc=<-2.4637 0.4233 0.9086>, rot=<0 0.2078 0 -0.9782>}
 FACIAL_L_Masseter {parent=FACIAL_C_FacialRoot, loc=<6.1094 1.2638 3.8656>, rot=<0.0708 0.113 0.0076 -0.991>}
 FACIAL_R_Masseter {parent=FACIAL_C_FacialRoot, loc=<-6.1097 1.2664 3.8573>, rot=<0.0708 -0.1128 -0.0078 -0.9911>}
 FACIAL_C_UnderChin {parent=FACIAL_C_FacialRoot, loc=<0.0029 8.8214 5.0925>, rot=<0.5011 0.0002 0 -0.8654>}
 FACIAL_L_UnderChin {parent=FACIAL_C_FacialRoot, loc=<3.6042 7.3733 3.112>, rot=<0.4038 -0.3546 0.0828 -0.8393>}
 FACIAL_R_UnderChin {parent=FACIAL_C_FacialRoot, loc=<-3.6027 7.3735 3.1052>, rot=<0.4037 0.3549 -0.0829 -0.8392>}
 FACIAL_C_TeethUpper {parent=FACIAL_C_FacialRoot, loc=<0.0018 4.6935 8.65>, rot=<0.0014 0.0001 -0.0001 -1>}
 FACIAL_C_LowerLipRotation {parent=FACIAL_C_FacialRoot, loc=<0.0004 1.6452 2.1186>, rot=<0.3434 0.0002 -0.0001 -0.9392>}
 FACIAL_C_MouthLower {parent=FACIAL_C_LowerLipRotation, loc=<0.0031 -1.2219 5.7247>, rot=<-0.342 0 0 -0.9397>}
 FACIAL_C_LipLower {parent=FACIAL_C_MouthLower, loc=<-0.0012 1.5892 2.1718>, rot=<0.0872 0 0 -0.9962>}
 FACIAL_L_LipLower {parent=FACIAL_C_MouthLower, loc=<1.0302 1.5273 2.0107>, rot=<0.0927 -0.1294 0.0272 -0.9869>}
 FACIAL_R_LipLower {parent=FACIAL_C_MouthLower, loc=<-1.0328 1.5261 2.0131>, rot=<0.0927 0.1294 -0.0272 -0.9869>}
 FACIAL_L_LipLowerOuter {parent=FACIAL_C_MouthLower, loc=<1.9764 0.9406 1.4446>, rot=<0.0913 -0.2321 0.0318 -0.9679>}
 FACIAL_R_LipLowerOuter {parent=FACIAL_C_MouthLower, loc=<-1.9805 0.9426 1.4466>, rot=<0.0913 0.2321 -0.0318 -0.9679>}
 FACIAL_C_Jaw {parent=FACIAL_C_FacialRoot, loc=<0.0004 1.6452 2.1186>, rot=<0.3434 0.0002 -0.0001 -0.9392>}
 FACIAL_C_Jawline {parent=FACIAL_C_Jaw, loc=<0.0006 2.7408 8.4736>, rot=<0.2586 0 0 -0.966>}
 FACIAL_L_ChinSide {parent=FACIAL_C_Jaw, loc=<3.0694 0.7976 7.6945>, rot=<-0.2208 -0.3738 0.0499 -0.8995>}
 FACIAL_R_ChinSide {parent=FACIAL_C_Jaw, loc=<-3.0636 0.8087 7.6912>, rot=<-0.2208 0.3738 -0.0499 -0.8995>}
 FACIAL_C_Chin {parent=FACIAL_C_Jaw, loc=<-0.001 1.1796 8.5132>, rot=<-0.2591 0 0 -0.9659>}
 FACIAL_L_Ear {parent=FACIAL_C_FacialRoot, loc=<6.6013 -0.2179 1.2843>, rot=<-0.0249 -0.0006 -0.0263 -0.9993>}
 FACIAL_R_Ear {parent=FACIAL_C_FacialRoot, loc=<-6.6025 -0.2165 1.2749>, rot=<-0.0249 0.0009 0.026 -0.9994>}
 FACIAL_L_Jawline {parent=FACIAL_C_FacialRoot, loc=<4.5333 5.654 4.2263>, rot=<0.2107 0.1578 -0.055 -0.9632>}
 FACIAL_R_Jawline {parent=FACIAL_C_FacialRoot, loc=<-4.5259 5.6506 4.2143>, rot=<0.2107 -0.1575 0.0548 -0.9632>}
 hair_aim_tgt {parent=spine_05, loc=<-0.123 -13.473 0>, rot=<-0.0093 -0.707 0.0093 -0.707>}
 clavicle_l {parent=spine_05, loc=<4.3575 -0.1466 -2.3223>, rot=<0.0465 -0.7289 -0.009 -0.6829>}
 upperarm_l {parent=clavicle_l, loc=<15.5111 0 0>, rot=<0.0226 -0.3234 -0.0088 -0.9459>}
 clavicle_r {parent=spine_05, loc=<4.3575 -0.1466 2.3223>, rot=<0.7289 0.0465 -0.6829 0.009>}
 upperarm_r {parent=clavicle_r, loc=<-15.5111 0 0>, rot=<0.0226 -0.3234 -0.0088 -0.9459>}
 Aim_Up_Fusile {type=Curve, loc=<0 0 0>, rot=<0 0 0 1>}
]
variables [Target Crouching {type=bool, default=False}]
