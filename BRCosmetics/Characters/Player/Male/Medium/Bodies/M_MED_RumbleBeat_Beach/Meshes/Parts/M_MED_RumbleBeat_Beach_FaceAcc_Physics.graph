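# Physics asset for the RumbleBeat Beach male medium FaceAcc mesh part, defining capsule bodies for three braids and a mullet.
# Each center, left and right braid chains four dynamic segments with twist-locked, swing-limited constraints.
# Base braid and mullet bodies have no constraints in the asset.
# Larger capsules on upper spine and clavicles form body volumes, apparently for the hair to collide against.
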PhysicsAsset /BRCosmetics/Characters/Player/Male/Medium/Bodies/M_MED_RumbleBeat_Beach/Meshes/Parts/M_MED_RumbleBeat_Beach_FaceAcc_Physics
bodies {
 base_braid01_cr [{shapeType=Capsule, length=0.5, radius=0.5, shapeName=base_braid01_cr_capsule, rotation=<rot 90 0 0>}]
 base_braid01_lf [{shapeType=Capsule, length=0.5, radius=0.5, shapeName=base_braid01_lf_capsule, rotation=<rot 90 0 0>}]
 base_braid01_rt [{shapeType=Capsule, length=0.5, radius=0.5, shapeName=base_braid01_rt_capsule, rotation=<rot 90 0 0>}]
 clavicle_l [{shapeType=Capsule, length=12, radius=8.5, shapeName=clavicle_l_capsule, center=<9.09 -1.5 -0.9095>, rotation=<rot 70 -0.001 -0.0011>}]
 clavicle_r [{shapeType=Capsule, length=5, radius=9.9, shapeName=clavicle_l_capsule, center=<-6.5643 0 0.7884>, rotation=<rot 70 -0.001 0>}]
 dyn_braid01_cr [{shapeType=Capsule, length=1.65, radius=1.5, shapeName=dyn_braid01_cr_capsule, center=<2.4816 0 0>, rotation=<rot 90 0 0>}]
 dyn_braid01_lf [{shapeType=Capsule, length=1.65, radius=1.5, shapeName=dyn_braid01_lf_capsule, center=<2.4816 0 0>, rotation=<rot 90 0 0>}]
 dyn_braid01_rt [{shapeType=Capsule, length=1.65, radius=1.5, shapeName=dyn_braid01_rt_capsule, center=<2.4816 0 0>, rotation=<rot 90 0 0>}]
 dyn_braid02_cr [{shapeType=Capsule, length=0.4, radius=1.5, shapeName=dyn_braid02_cr_capsule, center=<1.667 0 0>, rotation=<rot 90 0 0>}]
 dyn_braid02_lf [{shapeType=Capsule, length=2.2, radius=1.2, shapeName=dyn_braid02_lf_capsule, center=<2.0266 0 -0.1606>, rotation=<rot 85.0001 179.9999 -180>}]
 dyn_braid02_rt [{shapeType=Capsule, length=2.2, radius=1.2, shapeName=dyn_braid02_lf_capsule, center=<2.0266 0 -0.1606>, rotation=<rot 85.0001 179.9999 -180>}]
 dyn_braid03_cr [{shapeType=Capsule, length=2.25, radius=1.25, shapeName=dyn_braid03_cr_capsule, center=<2.5405 0 0>, rotation=<rot 90 0 0>}]
 dyn_braid03_lf [{shapeType=Capsule, length=2.3, radius=0.9, shapeName=dyn_braid03_lf_capsule, center=<2.4422 0 -0.2772>, rotation=<rot 90 0 0>}]
 dyn_braid03_rt [{shapeType=Capsule, length=2.3, radius=0.9, shapeName=dyn_braid03_lf_capsule, center=<2.4422 0 -0.2772>, rotation=<rot 90 -0.0001 0>}]
 dyn_braid04_cr [{shapeType=Capsule, length=3.3, radius=0.85, shapeName=dyn_braid04_cr_capsule, center=<2.5077 0 0.25>, rotation=<rot 80.0002 180 180>}]
 dyn_braid04_lf [{shapeType=Capsule, length=3, radius=0.86, shapeName=dyn_braid04_lf_capsule, center=<2.4855 0 0.3827>, rotation=<rot 80.0002 -180 -180>}]
 dyn_braid04_rt [{shapeType=Capsule, length=3.5, radius=0.85, shapeName=dyn_braid04_rt_capsule, center=<2.6485 0 0.1>, rotation=<rot 80.0002 180 180>}]
 dyn_mullet_cr [{shapeType=Capsule, length=3.5, radius=3, shapeName=dyn_mullet_cr_capsule, center=<3.4737 0 -2.4378>, rotation=<rot 0 0 90>}]
 dyn_mullet_lf [{shapeType=Capsule, length=3.5, radius=3, shapeName=dyn_mullet_lf_capsule, center=<3.4737 0 -1.2674>, rotation=<rot -0.0001 -0.0001 85>}]
 dyn_mullet_rt [{shapeType=Capsule, length=3.5, radius=3, shapeName=dyn_mullet_rt_capsule, center=<3.4737 0 -1.2674>, rotation=<rot -0.0001 -0.0001 85>}]
 spine_05 [
  {shapeType=Capsule, length=10, radius=13.5, shapeName=spine_05_capsule, center=<0.5474 0.219 0>}
  {shapeType=Capsule, length=8, radius=7.5, shapeName=spine_05_capsule, center=<9.6898 -0.5946 0>}
 ]
}
constraints [
 {bone1Name=dyn_braid02_cr, bone2Name=dyn_braid01_cr, swing1LimitDegrees=13, swing2LimitDegrees=50, twistLimitDegrees=0, twistMotion=Locked}
 {bone1Name=dyn_braid03_cr, bone2Name=dyn_braid02_cr, swing1LimitDegrees=13, swing2LimitDegrees=30, twistLimitDegrees=0, twistMotion=Locked}
 {bone1Name=dyn_braid04_cr, bone2Name=dyn_braid03_cr, swing1LimitDegrees=23, swing2LimitDegrees=30, twistLimitDegrees=0, twistMotion=Locked}
 {bone1Name=dyn_braid02_lf, bone2Name=dyn_braid01_lf, swing1LimitDegrees=12, swing2LimitDegrees=30, twistLimitDegrees=0, twistMotion=Locked}
 {bone1Name=dyn_braid03_lf, bone2Name=dyn_braid02_lf, swing1LimitDegrees=15, swing2LimitDegrees=30, twistLimitDegrees=0, twistMotion=Locked}
 {bone1Name=dyn_braid04_lf, bone2Name=dyn_braid03_lf, swing1LimitDegrees=25, swing2LimitDegrees=30, twistLimitDegrees=0, twistMotion=Locked}
 {bone1Name=dyn_braid02_rt, bone2Name=dyn_braid01_rt, swing1LimitDegrees=12, swing2LimitDegrees=45, twistLimitDegrees=0, twistMotion=Locked}
 {bone1Name=dyn_braid03_rt, bone2Name=dyn_braid02_rt, swing1LimitDegrees=8, swing2LimitDegrees=60, twistLimitDegrees=0, twistMotion=Locked}
 {bone1Name=dyn_braid04_rt, bone2Name=dyn_braid03_rt, swing1LimitDegrees=23, swing2LimitDegrees=30, twistLimitDegrees=0, twistMotion=Locked}
]
